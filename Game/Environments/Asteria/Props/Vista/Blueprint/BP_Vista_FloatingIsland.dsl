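; Vista floating island prop that scatters two tree variations over itself in the construction script; its event graph is empty.
; Each data table row (AsteriaFloatingIslandiCSV) supplies a position and normal, and seeded random rolls decide spawning and variation.
; Instances go to two InstancedStaticMesh components, aligned to the normal with random spin, with the first variation sunk along its up axis.
; Spawn chance scales with actor X scale, while tree scale divides by it.
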(event EventBeginPlay)

(event Collision|EventActorBeginOverlap
  (params (OtherActor :type Actor%20Object%20Reference)))

(event EventTick
  (params (DeltaSeconds :type Float%20%28single-precision%29)))

(fn ConstructionScript
  (bind _returnvalue (Transformation|GetActorScale3D))
  (bind _stream (Variables|Default|GetStream))
  (bind _returnvalue_2 (Math|Random|RandomFloatinRangefromStream _stream 0.0 3.6e+02))
  (bind _x (.x _returnvalue))
  (bind _returnvalue_3 (Math|Float|MakeLiteralFloat (Math|Float|SelectFloat (+ (Variables|Default|GetScale1) (* (Variables|Default|GetScale_variation) (Math|Random|RandomFloatinRangefromStream (Variables|Default|GetStream) -1.0 1.0))) (+ (Variables|Default|GetScale2) (* (Variables|Default|GetScale_variation) (Math|Random|RandomFloatinRangefromStream (Variables|Default|GetStream) -1.0 1.0))) (Variables|Default|GetCurrentTreeVariation))))
  (bind _returnvalue_4 (/ _returnvalue_3 _x))
  (bind _returnvalue_5 (Math|Vector|MakeVector _returnvalue_4 _returnvalue_4 _returnvalue_4))
  (Utilities|FlowControl|Sequence
    (:then_0
      (bind _outrownames (DataTable|GetDataTableRowNames (Variables|Default|GetData%20Table)))
      (bind _array_element (Utilities|Array|ForEachLoopwithBreak _outrownames)
        (:LoopBody
          (bind _returnvalue_6 (Utilities|GetDataTableRow (Variables|Default|GetData%20Table) _array_element))
          (Variables|Default|SetCurrentTreeVariation (> (Variables|Default|GetVariation) (Math|Random|RandomFloatinRangefromStream (Variables|Default|GetStream) 0.0 1.0)))
          (if (> (* (Variables|Default|GetSpawnChance) _x) (Math|Random|RandomFloatinRangefromStream (Variables|Default|GetStream) 0.0 1.0))
            (if (Variables|Default|GetCurrentTreeVariation)
              (Components|InstancedStaticMesh|AddInstance (Variables|Default|GetInstancedStaticMesh1) (Math|Transform|MakeTransform (+ (.location (Math|Transform|MakeTransform (Math|Vector|MakeVector (.x_20_C93EFB4E4D474825EF4022801F5E17C5 (Utilities|Struct|BreakAsteriaFloatingIslandiCSV _returnvalue_6)) (.y_22_2249D9184797030FDF396D816AB3DBA0 (Utilities|Struct|BreakAsteriaFloatingIslandiCSV _returnvalue_6)) (.z_18_46F26C4F4295E23E21BBB88A50FDEA73 (Utilities|Struct|BreakAsteriaFloatingIslandiCSV _returnvalue_6))) (Math|Rotator|CombineRotators (Math|Rotator|MakeRotfromZ (Math|Vector|MakeVector (Utilities|Struct|BreakAsteriaFloatingIslandiCSV _returnvalue_6) (.Rotationy_11_94ADCB5A46A0780CBD2666941CD7E362 (Utilities|Struct|BreakAsteriaFloatingIslandiCSV _returnvalue_6)) (.Rotationz_12_C7F059444A697EF95E1F7EB100805520 (Utilities|Struct|BreakAsteriaFloatingIslandiCSV _returnvalue_6)))) (Math|Vector|RotatorfromAxisandAngle (Math|Vector|MakeVector (Utilities|Struct|BreakAsteriaFloatingIslandiCSV _returnvalue_6) (.Rotationy_11_94ADCB5A46A0780CBD2666941CD7E362 (Utilities|Struct|BreakAsteriaFloatingIslandiCSV _returnvalue_6)) (.Rotationz_12_C7F059444A697EF95E1F7EB100805520 (Utilities|Struct|BreakAsteriaFloatingIslandiCSV _returnvalue_6))) _returnvalue_2)) _returnvalue_5)) (* (Math|Vector|GetUpVector (.rotation (Math|Transform|MakeTransform (Math|Vector|MakeVector (.x_20_C93EFB4E4D474825EF4022801F5E17C5 (Utilities|Struct|BreakAsteriaFloatingIslandiCSV _returnvalue_6)) (.y_22_2249D9184797030FDF396D816AB3DBA0 (Utilities|Struct|BreakAsteriaFloatingIslandiCSV _returnvalue_6)) (.z_18_46F26C4F4295E23E21BBB88A50FDEA73 (Utilities|Struct|BreakAsteriaFloatingIslandiCSV _returnvalue_6))) (Math|Rotator|CombineRotators (Math|Rotator|MakeRotfromZ (Math|Vector|MakeVector (Utilities|Struct|BreakAsteriaFloatingIslandiCSV _returnvalue_6) (.Rotationy_11_94ADCB5A46A0780CBD2666941CD7E362 (Utilities|Struct|BreakAsteriaFloatingIslandiCSV _returnvalue_6)) (.Rotationz_12_C7F059444A697EF95E1F7EB100805520 (Utilities|Struct|BreakAsteriaFloatingIslandiCSV _returnvalue_6)))) (Math|Vector|RotatorfromAxisandAngle (Math|Vector|MakeVector (Utilities|Struct|BreakAsteriaFloatingIslandiCSV _returnvalue_6) (.Rotationy_11_94ADCB5A46A0780CBD2666941CD7E362 (Utilities|Struct|BreakAsteriaFloatingIslandiCSV _returnvalue_6)) (.Rotationz_12_C7F059444A697EF95E1F7EB100805520 (Utilities|Struct|BreakAsteriaFloatingIslandiCSV _returnvalue_6))) _returnvalue_2)) _returnvalue_5))) -5e+02)) (.rotation (Math|Transform|MakeTransform (Math|Vector|MakeVector (.x_20_C93EFB4E4D474825EF4022801F5E17C5 (Utilities|Struct|BreakAsteriaFloatingIslandiCSV _returnvalue_6)) (.y_22_2249D9184797030FDF396D816AB3DBA0 (Utilities|Struct|BreakAsteriaFloatingIslandiCSV _returnvalue_6)) (.z_18_46F26C4F4295E23E21BBB88A50FDEA73 (Utilities|Struct|BreakAsteriaFloatingIslandiCSV _returnvalue_6))) (Math|Rotator|CombineRotators (Math|Rotator|MakeRotfromZ (Math|Vector|MakeVector (Utilities|Struct|BreakAsteriaFloatingIslandiCSV _returnvalue_6) (.Rotationy_11_94ADCB5A46A0780CBD2666941CD7E362 (Utilities|Struct|BreakAsteriaFloatingIslandiCSV _returnvalue_6)) (.Rotationz_12_C7F059444A697EF95E1F7EB100805520 (Utilities|Struct|BreakAsteriaFloatingIslandiCSV _returnvalue_6)))) (Math|Vector|RotatorfromAxisandAngle (Math|Vector|MakeVector (Utilities|Struct|BreakAsteriaFloatingIslandiCSV _returnvalue_6) (.Rotationy_11_94ADCB5A46A0780CBD2666941CD7E362 (Utilities|Struct|BreakAsteriaFloatingIslandiCSV _returnvalue_6)) (.Rotationz_12_C7F059444A697EF95E1F7EB100805520 (Utilities|Struct|BreakAsteriaFloatingIslandiCSV _returnvalue_6))) _returnvalue_2)) _returnvalue_5)) (.scale (Math|Transform|MakeTransform (Math|Vector|MakeVector (.x_20_C93EFB4E4D474825EF4022801F5E17C5 (Utilities|Struct|BreakAsteriaFloatingIslandiCSV _returnvalue_6)) (.y_22_2249D9184797030FDF396D816AB3DBA0 (Utilities|Struct|BreakAsteriaFloatingIslandiCSV _returnvalue_6)) (.z_18_46F26C4F4295E23E21BBB88A50FDEA73 (Utilities|Struct|BreakAsteriaFloatingIslandiCSV _returnvalue_6))) (Math|Rotator|CombineRotators (Math|Rotator|MakeRotfromZ (Math|Vector|MakeVector (Utilities|Struct|BreakAsteriaFloatingIslandiCSV _returnvalue_6) (.Rotationy_11_94ADCB5A46A0780CBD2666941CD7E362 (Utilities|Struct|BreakAsteriaFloatingIslandiCSV _returnvalue_6)) (.Rotationz_12_C7F059444A697EF95E1F7EB100805520 (Utilities|Struct|BreakAsteriaFloatingIslandiCSV _returnvalue_6)))) (Math|Vector|RotatorfromAxisandAngle (Math|Vector|MakeVector (Utilities|Struct|BreakAsteriaFloatingIslandiCSV _returnvalue_6) (.Rotationy_11_94ADCB5A46A0780CBD2666941CD7E362 (Utilities|Struct|BreakAsteriaFloatingIslandiCSV _returnvalue_6)) (.Rotationz_12_C7F059444A697EF95E1F7EB100805520 (Utilities|Struct|BreakAsteriaFloatingIslandiCSV _returnvalue_6))) _returnvalue_2)) _returnvalue_5))) false)
              (else
                (Components|InstancedStaticMesh|AddInstance (Variables|Default|GetInstancedStaticMesh2) (Math|Transform|MakeTransform (Math|Vector|MakeVector (.x_20_C93EFB4E4D474825EF4022801F5E17C5 (Utilities|Struct|BreakAsteriaFloatingIslandiCSV _returnvalue_6)) (.y_22_2249D9184797030FDF396D816AB3DBA0 (Utilities|Struct|BreakAsteriaFloatingIslandiCSV _returnvalue_6)) (.z_18_46F26C4F4295E23E21BBB88A50FDEA73 (Utilities|Struct|BreakAsteriaFloatingIslandiCSV _returnvalue_6))) (Math|Rotator|CombineRotators (Math|Rotator|MakeRotfromZ (Math|Vector|MakeVector (Utilities|Struct|BreakAsteriaFloatingIslandiCSV _returnvalue_6) (.Rotationy_11_94ADCB5A46A0780CBD2666941CD7E362 (Utilities|Struct|BreakAsteriaFloatingIslandiCSV _returnvalue_6)) (.Rotationz_12_C7F059444A697EF95E1F7EB100805520 (Utilities|Struct|BreakAsteriaFloatingIslandiCSV _returnvalue_6)))) (Math|Vector|RotatorfromAxisandAngle (Math|Vector|MakeVector (Utilities|Struct|BreakAsteriaFloatingIslandiCSV _returnvalue_6) (.Rotationy_11_94ADCB5A46A0780CBD2666941CD7E362 (Utilities|Struct|BreakAsteriaFloatingIslandiCSV _returnvalue_6)) (.Rotationz_12_C7F059444A697EF95E1F7EB100805520 (Utilities|Struct|BreakAsteriaFloatingIslandiCSV _returnvalue_6))) _returnvalue_2)) _returnvalue_5) false)))))))))
