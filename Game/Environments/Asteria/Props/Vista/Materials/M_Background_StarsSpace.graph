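# Unlit translucent vista backdrop material showing starry space that blends into drifting, time-of-day tinted clouds.
# Camera-vector lat-long UVs sample a distorted space texture, layered with tinted stars twinkling through panning noise masks.
# Height relative to camera drives a HeightLerp cloud transition and fades, while camera altitude shifts vertical offsets.
# Quality and feature-level switches drop distortion and extra star layers on cheaper platforms.
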Material /Game/Environments/Asteria/Props/Vista/Materials/M_Background_StarsSpace
graph {
 properties {blendMode=BLEND_Translucent, pixelDepthOffsetMode=PDOM_Legacy, shadingModel=MSM_Unlit, usedShadingModels=Unlit}
 nodes [
  n1 = NamedRerouteUsage()
  n2 = Multiply(A=@n1, B=ScalarParameter() {defaultValue=10, group=Stars, parameterName=Stars-StarTile})
  n3 = TextureSample(UVs=@n2) {
   samplerSource=SSM_Wrap_WorldGroupSettings
   samplerType=SAMPLERTYPE_LinearColor
   texture=/Game/Environments/Asteria/Props/LobbyIslandSpace/SkySphere/Textures/T_SkyDome_Stars_Sharp
  }
  n4 = ScalarParameter() {defaultValue=30, group=Stars, parameterName=Stars-StarMaskTileB}
  n5 = ScalarParameter() {defaultValue=0.5, group=Stars, parameterName=Stars-StarMaskTileBSpeed}
  n6 = TextureSample(UVs=Panner(
    Coordinate=Multiply(A=@n1, B=@n4)
    Time=MaterialFunctionCall(Speed=@n5) {materialFunction=&TimeWithSpeedVariable}.Time
   ) {speedX=0.1, speedY=0.1}) {
   samplerSource=SSM_Wrap_WorldGroupSettings
   samplerType=SAMPLERTYPE_LinearGrayscale
   texture=/Game/Global/Textures/Default/Tileable_Patterns/T_StormCloudDetail01_M
  }
  n7 = Power(Base=@n6.R, Exp=ScalarParameter() {defaultValue=8.3697, group=Stars, parameterName=Stars-StarMaskTileBPower})
  n8 = TextureSample(UVs=@n2) {
   samplerSource=SSM_Wrap_WorldGroupSettings
   texture=/Game/Environments/Asteria/Props/LobbyIslandSpace/SkySphere/Textures/T_SkyDome_Stars_Sharp_mask
  }
  n9 = TextureSample(UVs=Multiply(A=NamedRerouteUsage(), B=ScalarParameter() {defaultValue=20, parameterName=Param_20})) {
   samplerSource=SSM_Wrap_WorldGroupSettings
   texture=&T_SkyDome_Stars_Sharp_mask
  }
  n10 = LinearInterpolate(A=@n8.RGB, B=@n9.RGB, Alpha=Constant() {r=0.5})
  n11 = LinearInterpolate(
   A=Constant3Vector() {constant=<rgba 0 0.0473 1 0>}
   B=Constant3Vector() {constant=<rgba 1 0 0.083 0>}
   Alpha=ComponentMask(None=@n10) {r=true}
  )
  n12 = Constant3Vector() {constant=<rgba 0 1 0.0471 0>}
  n13 = ComponentMask(None=@n10) {b=true}
  n14 = Multiply(
   A=Multiply(A=Multiply(A=@n3.R, B=@n7), B=LinearInterpolate(A=@n11, B=@n12, Alpha=@n13))
   B=ScalarParameter() {defaultValue=2.5, parameterName=BGstar_brightness}
  )
  n15 = NamedRerouteUsage()
  n16 = Multiply(A=@n15, B=Constant() {r=5})
  n17 = TextureSample(UVs=@n16) {
   samplerSource=SSM_Wrap_WorldGroupSettings
   samplerType=SAMPLERTYPE_LinearColor
   texture=&T_SkyDome_Stars_Sharp
  }
  n18 = ScalarParameter() {defaultValue=4.5, group=Stars, parameterName=Stars-StarPower}
  n19 = TextureSample(UVs=Multiply(A=@n15, B=ScalarParameter() {defaultValue=20, group=Stars, parameterName=Stars-StarMaskTileA})) {
   samplerSource=SSM_Wrap_WorldGroupSettings
   samplerType=SAMPLERTYPE_LinearGrayscale
   texture=&T_StormCloudDetail01_M
  }
  n20 = Multiply(
   A=Multiply(A=Power(Base=@n17.R, Exp=@n18), B=ScalarParameter() {defaultValue=1.5, group=Stars, parameterName=Stars-StarMod})
   B=Power(Base=@n19.R, Exp=ScalarParameter() {defaultValue=6.6574, group=Stars, parameterName=Stars-StarMaskTileAPower})
  )
  n21 = ScalarParameter() {defaultValue=30, group=Stars, parameterName=Stars-StarMaskTileB}
  n22 = ScalarParameter() {defaultValue=0.5, group=Stars, parameterName=Stars-StarMaskTileBSpeed}
  n23 = TextureSample(UVs=Panner(
    Coordinate=Multiply(A=@n15, B=@n21)
    Time=MaterialFunctionCall(Speed=@n22) {materialFunction=&TimeWithSpeedVariable}.Time
   ) {speedX=0.1, speedY=0.1}) {
   samplerSource=SSM_Wrap_WorldGroupSettings
   samplerType=SAMPLERTYPE_LinearGrayscale
   texture=&T_StormCloudDetail01_M
  }
  n24 = Power(Base=@n23.R, Exp=ScalarParameter() {defaultValue=8.3697, group=Stars, parameterName=Stars-StarMaskTileBPower})
  n25 = ScalarParameter() {defaultValue=86.0193, group=Stars, parameterName=Stars-StarBrightness}
  n26 = Constant() {r=1051.9219}
  n27 = Constant3Vector() {constant=<rgba 0 0.0473 1 0>}
  n28 = Constant3Vector() {constant=<rgba 1 0 0.083 0>}
  n29 = TextureSample(UVs=Multiply(A=@n15, B=ScalarParameter() {defaultValue=10, group=Stars, parameterName=Stars-StarTile})) {
   samplerSource=SSM_Wrap_WorldGroupSettings
   texture=&T_SkyDome_Stars_Sharp_mask
  }
  n30 = TextureSample(UVs=@n16) {
   samplerSource=SSM_Wrap_WorldGroupSettings
   texture=&T_SkyDome_Stars_Sharp_mask
  }
  n31 = LinearInterpolate(A=@n29.RGB, B=@n30.RGB, Alpha=Constant() {r=0.284})
  n32 = ComponentMask(None=@n31) {r=true}
  n33 = Multiply(
   A=Multiply(A=Multiply(A=@n20, B=@n24), B=LinearInterpolate(A=@n25, B=@n26, Alpha=@n23.R))
   B=LinearInterpolate(A=LinearInterpolate(A=@n27, B=@n28, Alpha=@n32), B=Constant3Vector() {constant=<rgba 0 1 0.0471 0>}, Alpha=ComponentMask(None=@n31) {b=true})
  )
  n34 = NamedRerouteDeclaration(None=MaterialFunctionCall!(CameraVector) {materialFunction=/Engine/Functions/Engine_MaterialFunctions01/Texturing/CameraVectorToLatLongUV}.Uvs) {name=InfiniteUV, nodeColor=<rgba 0.6406 1 0 1>}
  n35 = VectorParameter() {parameterName=CloudDistortionSpeed, defaultValue=<rgba 0.01 0 0 1>}
  n36 = TextureSample(UVs=Panner(Speed=AppendVector(A=@n35.R, B=@n35.R)) {speedX=0.01}) {
   samplerSource=SSM_Wrap_WorldGroupSettings
   samplerType=SAMPLERTYPE_LinearGrayscale
   texture=&T_StormCloudDetail01_M
  }
  n37 = Add(
   A=@n34
   B=Multiply(
    A=ConstantBiasScale(None=@n36.R) {bias=-0.5, scale=2}
    B=ScalarParameter() {defaultValue=0.0035, parameterName=CloudDistortionAmount, sliderMin=0.001}
   ) {constB=0.01}
  )
  n38 = TextureCoordinate()
  n39 = TextureSample!(
   UVs=FeatureLevelSwitch(Default=QualitySwitch(Default=@n37, Low=@n34, Medium=@n34), ES3_1=@n34)
   Tex
   "DDX(UVs)"=DDX(Value=@n38)
   "DDY(UVs)"=DDY(Value=@n38)
   Apply View MipBias
  ) {mipValueMode=TMVM_Derivative, texture=/Game/Environments/Asteria/Props/LobbyIslandSpace/SkySphere/Textures/T_spaceBox_2k}
  n40 = Constant3Vector()
  n41 = ScalarParameter() {defaultValue=0.75, parameterName=Brightness}
  n42 = MaterialFunctionCall!(
   In=Desaturation(None=LinearInterpolate(A=@n39.RGB, B=@n40, Alpha=@n41), Fraction=ScalarParameter() {defaultValue=0.25, parameterName=saturation})
   Contrast=ScalarParameter() {defaultValue=0.001, parameterName=Contrast}
  ) {materialFunction=/Engine/Functions/Engine_MaterialFunctions01/ImageAdjustment/CheapContrast_RGB}
  n43 = Add(A=@n42.Result, B=Multiply(A=@n17.R, B=ScalarParameter() {defaultValue=0.2, parameterName=StarBrightnes Mobile}))
  n44 = FeatureLevelSwitch(Default=QualitySwitch(Default=Add(A=@n42.Result, B=@n33), Low=@n43), ES3_1=@n43)
  n45 = FeatureLevelSwitch(Default=QualitySwitch(Default=Add(A=@n44, B=@n14), Low=@n44, Medium=@n44), ES2=@n44, ES3_1=@n44)
  n46 = Multiply(A=ScalarParameter() {defaultValue=10, parameterName=SpaceBrightness}, B=ComponentMask(None=Saturate(None=@n45)) {b=true, g=true, r=true})
  Add(A=@n46, B=Multiply(A=ScalarParameter() {defaultValue=0.75, parameterName=FogAmount}, B=SkyAtmosphereDistantLightScatteredLuminance()))
  ComponentMask(None=SkyAtmosphereAerialPerspective!(World Position=WorldPosition().XYZ)) {b=true, g=true, r=true}
  Constant3Vector()
  n47 = Constant2Vector()
  n48 = ComponentMask(None=CameraPositionWS()) {b=true}
  n49 = ScalarParameter() {parameterName=WPOOffset}
  n50 = Saturate(None=Divide(A=Subtract(A=@n48, B=@n49), B=ScalarParameter() {defaultValue=80000, parameterName=WPOOffset_Mult}))
  LinearInterpolate(
   A=AppendVector(A=@n47, B=ScalarParameter() {defaultValue=-7613.0474, parameterName=WPO_Offset_MIn})
   B=AppendVector(A=@n47, B=ScalarParameter() {defaultValue=143633.0469, parameterName=WPO_Offset_Max})
   Alpha=@n50
  )
  n51 = SkyAtmosphereAerialPerspective(World Position=WorldPosition().XYZ)
  n52 = ScalarParameter() {defaultValue=1.8926, parameterName=GradientCloud}
  n53 = WorldPosition()
  n54 = ScalarParameter() {defaultValue=-35000, parameterName=PixelBlend_Offset_MIn}
  n55 = ScalarParameter() {defaultValue=-17000, parameterName=PixelBlend_Offset_Max}
  n56 = Subtract(
   A=Subtract(A=ComponentMask(None=@n53.XYZ) {b=true}, B=LinearInterpolate(A=@n54, B=@n55, Alpha=@n50))
   B=ComponentMask(None=CameraPositionWS()) {b=true}
  )
  n57 = Add(A=@n56, B=ScalarParameter() {defaultValue=152661.1094, parameterName=FadeOffsetTop2})
  n58 = ScalarParameter() {defaultValue=263831.125, parameterName=FadeDistanceTop2}
  n59 = OneMinus(None=Saturate(None=Divide(A=@n57, B=@n58)))
  n60 = Multiply(A=TextureCoordinate(), B=ScalarParameter() {defaultValue=8, parameterName=TilingTransitionClouds})
  n61 = AppendVector(
   A=Constant()
   B=LinearInterpolate(A=ScalarParameter() {defaultValue=0.456, parameterName=uv_Offset_MIn}, B=ScalarParameter() {parameterName=uv_Offset_Max}, Alpha=@n50)
  )
  n62 = ScalarParameter() {defaultValue=0.001, parameterName=CloudSpeedX}
  n63 = ScalarParameter() {defaultValue=-0.01, parameterName=CloudSpeedY}
  n64 = TextureSample!(UVs=Panner(Coordinate=Add(A=@n60, B=@n61), Speed=AppendVector(A=@n62, B=@n63)), Tex, Apply View MipBias) {texture=/Game/Athena/SafeZone/S11/T_TilingClouds_D}
  n65 = LinearInterpolate(
   A=ScalarParameter() {defaultValue=0.7971, parameterName=CloudContrast1}
   B=ScalarParameter() {defaultValue=0.9241, parameterName=CloudContrast2}
   Alpha=Add(A=Multiply(A=@n52, B=@n59), B=@n64.R)
  )
  n66 = DotProduct(A=SkyAtmosphereLightDirection(), B=VertexNormalWS())
  n67 = Saturate(None=Divide(A=Add(A=@n66)))
  n68 = LinearInterpolate(
   A=VectorParameter() {parameterName=CloudColor_Morning, defaultValue=<rgba 0.4786 0.2765 0.5052 1>}.RGB
   B=VectorParameter() {parameterName=CloudColor_Morning2, defaultValue=<rgba 0.5741 0.3337 0.6406 1>}.RGB
   Alpha=@n67
  )
  n69 = VectorParameter() {parameterName=CloudColor_Day, defaultValue=<rgba 1.8958 3.1713 4 1>}
  n70 = LinearInterpolate(
   A=VectorParameter() {parameterName=CloudColor_Evening, defaultValue=<rgba 0.625 0.2279 0.2966 1>}.RGB
   B=VectorParameter() {parameterName=CloudColor_Evening2, defaultValue=<rgba 0.599 0.5664 0.5023 1>}.RGB
   Alpha=@n67
  )
  n71 = VectorParameter() {parameterName=CloudColor_Night, defaultValue=<rgba 0 0 0 1>}
  n72 = ScalarParameter() {defaultValue=0.639, parameterName=ContrastTransition}
  n73 = Power(
   Base=MaterialFunctionCall(Transition Phase=@n59, Height Texture=@n64.R, Contrast=@n72) {materialFunction=&HeightLerp}.Lerp Alpha No Contrast
   Exp=@n72
  )
  n74 = Multiply(A=@n64.R, B=ScalarParameter() {defaultValue=-0.144, parameterName=MinCloudOverlay})
  LinearInterpolate(
   A=LinearInterpolate(
    A=MaterialFunctionCall(Color_Morning=@n68, Color_Day=@n69.RGB, Color_Evening=@n70, Color_Night=@n71.RGB) {materialFunction=&Color_Time_of_Day}.Result
    B=ComponentMask(None=@n51) {b=true, g=true, r=true}
    Alpha=Saturate(None=@n65)
   )
   B=@n46
   Alpha=Saturate(None=Add(A=@n73, B=@n74))
  )
  n75 = Divide(
   A=Add(A=@n56, B=ScalarParameter() {defaultValue=-50897.543, parameterName=FadeOffsetTop_HL})
   B=ScalarParameter() {defaultValue=80398.7891, parameterName=FadeDistanceTop_HL}
  )
  n76 = ScalarParameter() {parameterName=FadeFalloff_HL}
  n77 = ScalarParameter() {defaultValue=46422.3789, parameterName=FadeDistanceTop}
  n78 = Power(
   Base=Add(A=Divide(A=@n56, B=@n77), B=ScalarParameter() {defaultValue=-1.3643, parameterName=FadeOffsetTop})
   Exp=ScalarParameter() {defaultValue=1.3539, parameterName=FadeFalloff}
  )
  Multiply(
   A=OneMinus(None=MaterialFunctionCall(Transition Phase=@n75, Height Texture=@n64.R, Contrast=@n76) {materialFunction=&HeightLerp}.Lerp Alpha No Contrast)
   B=Saturate(None=OneMinus(None=@n78))
  )
 ]
}
