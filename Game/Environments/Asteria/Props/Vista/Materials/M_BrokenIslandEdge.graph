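# Nanite-enabled material for Asteria vista broken island edge props, blending world-aligned cliff with a second layer by a mask texture.
# A static switch makes that layer flat-colored grass/rock for floating islands, or dirt blended into water by world-space soil mask and height.
# The mesh normal map is blended over the detail normals, and mask green darkens base color as AO on non-floating islands.
# Grass-over-dirt, decontrast and slope-blend nodes are left unconnected.
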Material /Game/Environments/Asteria/Props/Vista/Materials/M_BrokenIslandEdge
graph {
 properties {bUseMaterialAttributes=true, bUsedWithNanite=true, pixelDepthOffsetMode=PDOM_Legacy, usedShadingModels=Default Lit}
 nodes [
  n1 = VectorParameter() {parameterName=TintDirt1, defaultValue=<rgba 0.2964 0.4375 0 1>}
  n2 = VectorParameter() {parameterName=TintDirt2, defaultValue=<rgba 0.1354 0.1001 0.0633 1>}
  n3 = WorldPosition()
  n4 = TextureSample(UVs=Divide(A=ComponentMask(None=@n3.XYZ) {g=true, r=true}, B=ScalarParameter() {defaultValue=50373.9102, parameterName=SizeWaterSoilMask})) {texture=/Game/Environments/Asteria/Props/Vista/Textures/T_WaterSoilMask}
  n5 = Saturate(None=LinearInterpolate(
    A=ScalarParameter() {defaultValue=0.6238, parameterName=ContrastEdgeMask_1}
    B=ScalarParameter() {defaultValue=5.8745, parameterName=ContrastEdgeMask_2}
    Alpha=@n4.R
   ))
  n6 = TextureSampleParameter2D() {parameterName=Mask, samplerType=SAMPLERTYPE_Masks, texture=/Game/Environments/Asteria/Props/Vista/Textures/BrokenIslandEdge_Mask}
  n7 = ScalarParameter() {defaultValue=3104.9097, parameterName=SizeDirt}
  n8 = AppendVector(A=AppendVector(A=@n7, B=@n7), B=@n7)
  n9 = Multiply(
   A=LinearInterpolate(
    A=LinearInterpolate(A=@n1.RGB, B=@n2.RGB, Alpha=@n5)
    B=VectorParameter() {parameterName=TintDirtEdge, defaultValue=<rgba 0.3177 0.3118 0.3061 1>}.RGB
    Alpha=@n6.R
   )
   B=MaterialFunctionCall(TextureObject=TextureObject() {texture=/Game/Environments/Asteria/Terrain/Textures/T_Asteria_DirtGravel_01_D}, TextureSize=@n8) {materialFunction=&WorldAlignedTexture}.Z Texture
  )
  n10 = ScalarParameter() {defaultValue=-0.7399, parameterName=GrassrMask_C1}
  n11 = ScalarParameter() {defaultValue=8.3722, parameterName=GrassMask_C2}
  LinearInterpolate(
   A=@n9
   B=VectorParameter() {parameterName=GrassColor, defaultValue=<rgba 0.0347 0.0781 0.0069 1>}.RGB
   Alpha=Saturate(None=LinearInterpolate(A=@n10, B=@n11, Alpha=@n4.G))
  )
  n12 = VectorParameter() {parameterName=TintCliff, defaultValue=<rgba 0.2188 0.1273 0.0897 1>}
  n13 = VectorParameter() {parameterName=TintCliffEdge, defaultValue=<rgba 0.3385 0.2111 0.1587 1>}
  n14 = TextureObject() {texture=/Game/Environments/Asteria/Terrain/Textures/T_Asteria_QuarryCliff_01_D}
  n15 = ScalarParameter() {defaultValue=2899.0327, parameterName=SizeCliff}
  n16 = AppendVector(A=AppendVector(A=@n15, B=@n15), B=@n15)
  n17 = ScalarParameter() {defaultValue=-0.2848, parameterName=ContrastCliffProjection}
  n18 = MaterialFunctionCall(Float3=MaterialFunctionCall(
    TextureObject=TextureObject() {samplerType=SAMPLERTYPE_Masks, texture=/Game/Environments/Asteria/Terrain/Textures/T_Asteria_RockCliff_01_S}
    TextureSize=@n16
    ProjectionTransitionContrast=@n17
   ) {materialFunction=&WorldAlignedTexture}.XY Texture) {materialFunction=&BreakOutFloat3Components}
  n19 = MaterialFunctionCall(
   TextureObject=TextureObject() {samplerType=SAMPLERTYPE_Normal, texture=/Game/Environments/Asteria/Terrain/Textures/T_Asteria_QuarryCliff_01_N}
   TextureSize=@n16
   ProjectionTransitionContrast=@n17
  ) {materialFunction=&WorldAlignedNormal}
  n20 = SetMaterialAttributes!(
   MaterialAttributes
   Base Color=Multiply(
    A=LinearInterpolate(A=@n12.RGB, B=@n13.RGB, Alpha=@n6.R)
    B=MaterialFunctionCall(TextureObject=@n14, TextureSize=@n16, ProjectionTransitionContrast=@n17) {materialFunction=&WorldAlignedTexture}.XY Texture
   )
   Specular=@n18.R
   Roughness=@n18.R
   Normal=MaterialFunctionCall(Normal=@n19.XYZ Texture, Flatness=ScalarParameter() {defaultValue=0.25, parameterName=CliffNormal}) {materialFunction=&FlattenNormal}.Result
  ) {attributeSetTypes [
    #1
    #2
    #3
    #4
   ]}
  n21 = MaterialFunctionCall(
   TextureObject=TextureObject() {samplerType=SAMPLERTYPE_Normal, texture=/Game/Environments/Asteria/Terrain/Textures/T_Asteria_FrozenGravel_01_N}
   TextureSize=@n8
  ) {materialFunction=&WorldAlignedNormal}
  n22 = SetMaterialAttributes(
   Base Color=LinearInterpolate(
    A=VectorParameter() {parameterName=Color_Grass, defaultValue=<rgba 0.0444 0.1406 0 1>}.RGB
    B=VectorParameter() {parameterName=Color_Grass_2, defaultValue=<rgba 0.1354 0.112 0.0226 1>}.RGB
    Alpha=@n6.R
   )
   Roughness=ScalarParameter() {defaultValue=0.92, parameterName=Roughness_Grass}
   Normal=@n21.Z Texture
  ) {attributeSetTypes [
    #1
    #2
    #3
    #4
   ]}
  n23 = SetMaterialAttributes(
   Base Color=LinearInterpolate(
    A=VectorParameter() {parameterName=Color_Cliff, defaultValue=<rgba 0.0833 0.0664 0.0463 1>}.RGB
    B=VectorParameter() {parameterName=Color_Cliff_2, defaultValue=<rgba 0.0312 0.0296 0.0273 1>}.RGB
    Alpha=@n6.R
   )
   Roughness=ScalarParameter() {defaultValue=0.92, parameterName=Roughness_Cliff}
   Normal=@n19.XYZ Texture
  ) {attributeSetTypes [
    #1
    #2
    #3
    #4
   ]}
  n24 = BlendMaterialAttributes(A=@n22, B=@n23, Alpha=LinearInterpolate(Alpha=@n6.G))
  n25 = MaterialFunctionCall(Float3=MaterialFunctionCall(
    TextureObject=TextureObject() {samplerType=SAMPLERTYPE_Masks, texture=/Game/Environments/Asteria/Terrain/Textures/T_Asteria_FrozenGravel_01_S}
    TextureSize=@n8
   ) {materialFunction=&WorldAlignedTexture}.Z Texture) {materialFunction=&BreakOutFloat3Components}
  n26 = ScalarParameter() {defaultValue=0.2986, parameterName=SoilWaterRoughness_1}
  n27 = ScalarParameter() {parameterName=SoilWaterRoughness_2}
  n28 = SetMaterialAttributes(Base Color=@n9, Specular=@n25.R, Roughness=Multiply(A=LinearInterpolate(A=@n26, B=@n27, Alpha=@n4.R), B=@n25.B), Normal=@n21.Z Texture) {attributeSetTypes [
    #1
    #2
    #3
    #4
   ]}
  n29 = WorldPosition()
  n30 = Divide(
   A=Add(A=ComponentMask(None=@n29.XYZ) {b=true}, B=ScalarParameter() {defaultValue=-700, parameterName=Water_Level})
   B=ScalarParameter() {defaultValue=2752.0571, parameterName=Water_DistanceBlend}
  )
  n31 = SetMaterialAttributes(
   Base Color=LinearInterpolate(
    A=VectorParameter() {parameterName=BaseColor, defaultValue=<rgba 0 0.0122 0.02 0.1>}.RGB
    B=VectorParameter() {parameterName=WaterColor, defaultValue=<rgba 0.0938 0.0751 0.0482 0.1>}.RGB
    Alpha=Saturate(None=@n30)
   )
   Specular=Constant() {r=0.5}
   Roughness=ScalarParameter() {parameterName=Water_Roughness}
   Normal=Constant3Vector() {constant=<rgba 0 0 1 0>}
  ) {attributeSetTypes [
    #1
    #2
    #3
    #4
   ]}
  n32 = Saturate(None=LinearInterpolate(
    A=ScalarParameter() {defaultValue=-0.1481, parameterName=WaterMask_C1}
    B=ScalarParameter() {defaultValue=1.5348, parameterName=WaterMask_C2}
    Alpha=@n4.R
   ))
  n33 = Add(A=Multiply(A=ScalarParameter() {defaultValue=0.7382, parameterName=WaterInfluence}, B=@n32), B=@n30)
  n34 = Saturate(None=LinearInterpolate(
    A=ScalarParameter() {defaultValue=1, parameterName=WaterTransition_1}
    B=ScalarParameter() {defaultValue=-1.6618, parameterName=WaterTransition_2}
    Alpha=Saturate(None=@n33)
   ))
  n35 = BlendMaterialAttributes(A=@n28, B=@n31, Alpha=Saturate(None=Add(A=@n32, B=@n34)))
  n36 = GetMaterialAttributes(None=BlendMaterialAttributes(A=@n20, B=StaticSwitchParameter(True=@n24, False=@n35) {parameterName=IsFloatingIsland}, Alpha=@n6.B)) {attributeGetTypes [#4, #1]}
  LinearInterpolate(
   A=@n36.Base Color
   B=VectorParameter() {parameterName=Decontrast, defaultValue=<rgba 0.1302 0.0698 0.0228 1>}.RGB
   Alpha=ScalarParameter() {defaultValue=0.25, parameterName=ColorFade}
  )
  MaterialFunctionCall() {materialFunction=&BreakOutFloat3Components}
  n37 = VertexNormalWS()
  Saturate(None=LinearInterpolate(
    A=ScalarParameter() {defaultValue=-2.916, parameterName=Blend1}
    B=ScalarParameter() {defaultValue=2, parameterName=Blend2}
    Alpha=ComponentMask(None=@n37) {b=true}
   ))
  n38 = TextureSampleParameter2D() {parameterName=Normal, samplerType=SAMPLERTYPE_Normal, texture=/Game/Environments/Asteria/Props/Vista/Textures/BrokenIslandEdge_Normal}
  n39 = Saturate(None=Add(A=@n6.G, B=ScalarParameter() {defaultValue=0.125, parameterName=AO}))
  SetMaterialAttributes!(
   MaterialAttributes=@n36.MaterialAttributes
   Normal=MaterialFunctionCall(BaseNormal=@n38.RGB, AdditionalNormal=@n36.Normal) {materialFunction=&BlendAngleCorrectedNormals}.Result
   Base Color=StaticSwitchParameter(True=@n36.Base Color, False=Multiply(A=@n39, B=@n36.Base Color)) {parameterName=IsFloatingIsland}
  ) {attributeSetTypes [#4, #1]}
 ]
}
