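# Unlit additive material for world-edge vista fog, animated with scrolling cloud noise.
# Time-of-day cloud colors are multiplied by sky aerial perspective and brightness, then desaturated.
# A light-beam falloff times panned cloud texture feeds a depth fade, scaled by a camera-height fade.
# When that height fade reaches zero, a negated world-position term becomes active, apparently an offset collapsing the mesh to the origin.
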Material /Game/Environments/Asteria/Props/Vista/Materials/M_WorldEdge_Fog
graph {
 properties {bUseTranslucencyVertexFog=false, blendMode=BLEND_Additive, pixelDepthOffsetMode=PDOM_Legacy, shadingModel=MSM_Unlit, usedShadingModels=Unlit}
 nodes [
  n1 = ComponentMask(None=SkyAtmosphereAerialPerspective!(World Position=WorldPosition().XYZ)) {b=true, g=true, r=true}
  n2 = MaterialFunctionCall(
   Color_Morning=VectorParameter() {parameterName=CloudColor_Morning, defaultValue=<rgba 0.0629 0.0376 0.099 1>}.RGB
   Color_Day=VectorParameter() {parameterName=CloudColor_Day, defaultValue=<rgba 0.1985 1 0.8801 1>}.RGB
   Color_Evening=VectorParameter() {parameterName=CloudColor_Evening, defaultValue=<rgba 0.208 0.0488 0.0885 1>}.RGB
   Color_Night=VectorParameter() {parameterName=CloudColor_Night, defaultValue=<rgba 0.3229 0.4004 1 1>}.RGB
  ) {materialFunction=&Color_Time_of_Day}
  Desaturation(
   None=Multiply(A=Multiply(A=@n2.Result, B=@n1), B=ScalarParameter() {defaultValue=1, parameterName=Brightness})
   Fraction=ScalarParameter() {defaultValue=0.306, parameterName=Desat}
  )
  n3 = TextureSample() {samplerType=SAMPLERTYPE_Grayscale, texture=/Engine/EngineVolumetrics/LightBeam/Materials/T_EV_LightBeam_Falloff_02}
  n4 = Panner(
   Coordinate=Multiply(A=TextureCoordinate(), B=ScalarParameter() {defaultValue=2.1887, parameterName=TilingTransitionClouds})
   Speed=AppendVector(A=ScalarParameter() {defaultValue=0.01, parameterName=CloudSpeedX}, B=ScalarParameter() {defaultValue=-0.02, parameterName=CloudSpeedY})
  )
  n5 = Saturate(None=LinearInterpolate(
    A=ScalarParameter() {defaultValue=0.816, parameterName=Min}
    B=ScalarParameter() {defaultValue=1, parameterName=Max}
    Alpha=TextureSample(UVs=@n4) {texture=/Game/Athena/SafeZone/S11/T_TilingClouds_D}.R
   ))
  MaterialFunctionCall(Transition Phase=@n3.R, Height Texture=@n5) {materialFunction=&HeightLerp}
  n6 = ComponentMask(None=CameraPositionWS()) {b=true}
  n7 = ScalarParameter() {defaultValue=12000, parameterName=HeightFade_Offset}
  n8 = Saturate(None=Divide(A=Subtract(A=@n6, B=@n7), B=ScalarParameter() {defaultValue=15000, parameterName=HeightFade_Fade}))
  Multiply(A=DepthFade(Opacity=Multiply(A=@n3.R, B=@n5), FadeDistance=ScalarParameter() {defaultValue=55000, parameterName=DepthFade}), B=@n8)
  Multiply(A=Multiply(A=WorldPosition().XYZ) {constB=-1}, B=OneMinus(None=Ceil(None=@n8)))
  VectorParameter() {parameterName=Color, defaultValue=<rgba 1.5 1.5 1.5 1>}
 ]
}
